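# Base UI Kit tabs widget that groups tabs into expandable categories, built from category view models.
# AddButtonToGrid override wraps each header button in a category container with inner tabs that inherit layout properties; allowed empty categories remain plain buttons.
# Container events are relayed through dispatchers and keep selection exclusive, collapsing other categories (header clicks can allow several open).
# Gamepad input enables select-on-focus and first-tab selection on expand; other inputs allow click-to-collapse, highlighting headers of collapsed selections.
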WidgetBlueprint /FortUILibrary/Systems/Tabs/WBP_UIKit_Tabs_Categories_Base
graphs [
 AddButtonToGrid {nodes [
   n1 = AddButtonToGrid!() -> (then, Widget:Widget Object Reference, Index:int=0) {is=title K2Node_FunctionEntry}
   > n2 = Set Category(Category:UI Kit Tab Group Category View Model Object Reference=GetCategoryViewModelByIndex!(self, Index:int=@n1.Index) -> (Category:UI Kit Tab Group Category View Model Object Reference=UIKit Tabs)) -> (Output_Get:UI Kit Tab Group Category View Model Object Reference)
   > Branch(Condition=IsCategoryViewModelValid!(self, CategoryVM:UI Kit Tab Group Category View Model Object Reference=@n2) -> (IsValid:bool))
   n3 = Get Tabs!(self:UI Kit Tab Group Category View Model Object Reference=@n2) -> (Tabs:Array of UI Kit Tab Group Button View Model Object References)
   > n4 = Branch(Condition=AND Boolean(
     A=Get AllowEmptyCategories() -> (AllowEmptyCategories:bool)
     B=Is Empty!(TargetArray:Array of UI Kit Tab Group Button View Model Object References=@n3) -> (ReturnValue:bool)
    ))
   > Parent: AddButtonToGrid!(execute, Widget:Widget Object Reference=Get Widget() -> (Widget:Widget Object Reference), Index:int=Get Index() -> (Index:int=0)) -> (then) {is=title K2Node_CallParentFunction}
   n5 = Cast To CommonButtonBase!(execute=@n4.else, Object=Get Widget()) -> (then, CastFailed, AsCommon Button Base:Common Button Base Object Reference) {is=title K2Node_DynamicCast}
   > n6 = CreateCategoryContainer!(execute, self) -> (then, Container:WBP UIKit Tab Category Base Object Reference)
   > Parent: AddButtonToGrid(Widget=@n6, Index=Get Index() -> (Index=0))
   > InitContainer!(
    execute
    self:WBP UIKit Tab Category Base Object Reference=@n6
    HeaderTabButton:Common Button Base Object Reference=@n5
    CategoryVM:UI Kit Tab Group Category View Model Object Reference=Get Category() -> (Category:UI Kit Tab Group Category View Model Object Reference)
   ) -> (then)
  ]}
 CheckCategoryChanged {nodes [
   n1 = CheckCategoryChanged!() -> (then, CategoryID:Name) {is=title K2Node_FunctionEntry}
   n2 = Get Previous Selected Category() -> (Previous Selected Category:Name=None)
   > Branch(Condition="Not Equal (Name)"(A=@n1, B=@n2) -> (ReturnValue=~))
   > RefreshCategorySelectableState!(execute, self, CategoryID:Name=@n2, InputType:ECommonInputType Enum=GetCurrentInputType(self=Get CommonInputSubsystem())) -> (then)
   > ClearCategorySelection!(execute, self, CategoryID:Name=@n2) -> (then)
   > Set Previous Selected Category(Previous Selected Category:Name=Get CategoryID() -> (CategoryID:Name=None)) -> (Output_Get:Name=None)
  ]}
 ClearCategorySelection {nodes [
   n1 = ClearCategorySelection!() -> (then, CategoryID:Name) {is=title K2Node_FunctionEntry}
   > n2 = Set CategoryContainer(CategoryContainer:WBP UIKit Tab Category Base Object Reference=GetCategoryContainer!(self, CategoryID:Name=@n1) -> (Container:WBP UIKit Tab Category Base Object Reference)) -> (Output_Get:WBP UIKit Tab Category Base Object Reference)
   > Is Valid(InputObject=@n2)
   > ClearCategorySelection!(execute, self:WBP UIKit Tab Category Base Object Reference=@n2) -> (then) {is=title K2Node_CallFunction}
  ]}
 CollapseOthers {nodes [
   n1 = CollapseOthers!() -> (then, ExceptCategoryID:Name) {is=title K2Node_FunctionEntry}
   > n2 = For Each Loop(Array:Array of WBP UIKit Tab Category Base Object References=Get CategoryContainers() -> (CategoryContainers:Array of WBP UIKit Tab Category Base Object References)) -> (Array Element:WBP UIKit Tab Category Base Object Reference)
   > Branch(Condition="Not Equal (Name)"(A=Get CategoryID!(self:WBP UIKit Tab Category Base Object Reference=@n2) -> (CategoryID=None), B=@n1) -> (ReturnValue=~))
   > SetIsExpanded!(
    execute
    self:WBP UIKit Tab Category Base Object Reference=@n2
    InIsExpanded:bool=false
    SkipAnim:bool=Get Skip Expand Animation() -> (Skip Expand Animation:bool)
   ) -> (then)
  ]}
 CopyTabsProps {nodes [
   CopyTabsProps!() -> (then, Target:WBP UIKit Tabs Base Object Reference) {is=title K2Node_FunctionEntry}
   > n1 = Sequence!(execute) -> (then_0, then_1, then_2:Exec, then_3:Exec, then_4:Exec)
   n2 = Self-Reference()
   > Set IsVertical!(
    execute
    IsVertical:bool=Get IsVertical!(self:WBP UIKit Tabs Base Object Reference=@n2) -> (IsVertical:bool)
    self:WBP UIKit Tabs Base Object Reference=Get Target() -> (Target:WBP UIKit Tabs Base Object Reference)
   ) -> (then, Output_Get:bool)
   > Set IsNaturalNavigation!(execute, IsNaturalNavigation:bool=true, self:WBP UIKit Tabs Base Object Reference=Get Target()) -> (then, Output_Get:bool)
   > Set TabGroupHorizontalAlignment!(
    execute
    TabGroupHorizontalAlignment:EHorizontalAlignment Enum=Get TabGroupHorizontalAlignment!(self:WBP UIKit Tabs Base Object Reference=@n2) -> (TabGroupHorizontalAlignment:EHorizontalAlignment Enum=HAlign_Fill)
    self:WBP UIKit Tabs Base Object Reference=Get Target()
   ) -> (then, Output_Get:EHorizontalAlignment Enum=HAlign_Fill)
   > Set TabGroupVerticalAlignment!(
    execute
    TabGroupVerticalAlignment:EVerticalAlignment Enum=Get TabGroupVerticalAlignment!(self:WBP UIKit Tabs Base Object Reference=@n2) -> (TabGroupVerticalAlignment:EVerticalAlignment Enum=VAlign_Fill)
    self:WBP UIKit Tabs Base Object Reference=Get Target()
   ) -> (then, Output_Get:EVerticalAlignment Enum=VAlign_Fill)
   > Set TabGroupPadding!(
    execute
    TabGroupPadding:Margin Structure=Get TabGroupPadding!(self:WBP UIKit Tabs Base Object Reference=@n2) -> (TabGroupPadding:Margin Structure)
    self:WBP UIKit Tabs Base Object Reference=Get Target()
   ) -> (then, Output_Get:Margin Structure)
   Set ButtonClass!(
    execute=@n1.then_1
    ButtonClass:WBP UIKit Button Generic Base Class Reference=Get CategoryButtonClass() -> (CategoryButtonClass:WBP UIKit Button Generic Base Class Reference)
    self:WBP UIKit Tabs Base Object Reference=Get Target()
   ) -> (then, Output_Get:WBP UIKit Button Generic Base Class Reference)
   > Set MinButtonWidth!(execute, MinButtonWidth:int=Get MinButtonWidth() -> (MinButtonWidth:int=0), self:WBP UIKit Tabs Base Object Reference=Get Target()) -> (then, Output_Get:int=0)
   > Set MinButtonHeight!(execute, MinButtonHeight:int=Get MinButtonHeight() -> (MinButtonHeight:int=0), self:WBP UIKit Tabs Base Object Reference=Get Target()) -> (then, Output_Get:int=0)
   > Set ButtonHorizontalAlignment!(
    execute
    ButtonHorizontalAlignment:EHorizontalAlignment Enum=Get ButtonHorizontalAlignment!(self:WBP UIKit Tabs Base Object Reference=@n2) -> (ButtonHorizontalAlignment:EHorizontalAlignment Enum=HAlign_Fill)
    self:WBP UIKit Tabs Base Object Reference=Get Target()
   ) -> (then, Output_Get:EHorizontalAlignment Enum=HAlign_Fill)
   > Set ButtonVerticalAlignment!(
    execute
    ButtonVerticalAlignment:EVerticalAlignment Enum=Get ButtonVerticalAlignment!(self:WBP UIKit Tabs Base Object Reference=@n2) -> (ButtonVerticalAlignment:EVerticalAlignment Enum=VAlign_Fill)
    self:WBP UIKit Tabs Base Object Reference=Get Target()
   ) -> (then, Output_Get:EVerticalAlignment Enum=VAlign_Fill)
   > Set TabSpacing!(
    execute
    TabSpacing:double=Get TabSpacing!(self:WBP UIKit Tabs Base Object Reference=@n2) -> (TabSpacing:double)
    self:WBP UIKit Tabs Base Object Reference=Get Target()
   ) -> (then, Output_Get:double)
   > Set IconSize!(
    execute
    IconSize:Vector 2D Structure=Get IconSize() -> (IconSize:Vector 2D Structure)
    self:WBP UIKit Tabs Base Object Reference=Get Target()
   ) -> (then, Output_Get:Vector 2D Structure)
   > Set SelectionAlignment!(
    execute
    SelectionAlignment:E_UI_Alignment Enum=Get SelectionAlignment!(self:WBP UIKit Tabs Base Object Reference=@n2) -> (SelectionAlignment:E_UI_Alignment Enum=NewEnumerator0)
    self:WBP UIKit Tabs Base Object Reference=Get Target()
   ) -> (then, Output_Get:E_UI_Alignment Enum=NewEnumerator0)
   SetInputActionsType!(
    execute=@n1.then_2
    self:WBP UIKit Tabs Base Object Reference=Get Target()
    InputActions:E_UI_TabsInputActionType Enum=NewEnumerator2
    InputActionsWithMouseScroll:bool=false
   ) -> (then)
   Set UseDividers!(execute=@n1.then_3, UseDividers:bool=Get UseDividers() -> (UseDividers:bool), self:WBP UIKit Tabs Base Object Reference=Get Target()) -> (then, Output_Get:bool)
   > Set DividerClass!(
    execute
    DividerClass:WBP UIKit Divider Base Class Reference=Get DividerClass() -> (DividerClass:WBP UIKit Divider Base Class Reference)
    self:WBP UIKit Tabs Base Object Reference=Get Target()
   ) -> (then, Output_Get:WBP UIKit Divider Base Class Reference)
   > Set DividerSpacing!(
    execute
    DividerSpacing:double=Get DividerSpacing!(self:WBP UIKit Tabs Base Object Reference=@n2) -> (DividerSpacing:double)
    self:WBP UIKit Tabs Base Object Reference=Get Target()
   ) -> (then, Output_Get:double)
   Set UseBackplate!(execute=@n1.then_4, UseBackplate:bool=false, self:WBP UIKit Tabs Base Object Reference=Get Target()) -> (then, Output_Get:bool)
   Get UseBackplate() -> (UseBackplate:bool)
   Get IsNaturalNavigation!(self:WBP UIKit Tabs Base Object Reference=@n2) -> (IsNaturalNavigation:bool)
  ]}
 CreateCategoryContainer {nodes [
   CreateCategoryContainer() {is=title K2Node_FunctionEntry}
   > n1 = Create Widget!(
    execute:Exec
    Class:User Widget Class Reference=Get CategoryClass() -> (CategoryClass:WBP UIKit Tab Category Base Class Reference)
    OwningPlayer:Player Controller Object Reference=GetOwningPlayer()
   ) -> (then:Exec, ReturnValue:WBP UIKit Tab Category Base Object Reference)
   > n2 = Set CategoryWidget(CategoryWidget:WBP UIKit Tab Category Base Object Reference=@n1) -> (Output_Get:WBP UIKit Tab Category Base Object Reference)
   > Add(
    TargetArray:Array of WBP UIKit Tab Category Base Object References=Get CategoryContainers()
    NewItem:"WBP UIKit Tab Category Base Object Reference (by ref)"=@n2
   ) {is=title K2Node_CallArrayFunction}
   n3 = Get CategoryWidget() -> (CategoryWidget:WBP UIKit Tab Category Base Object Reference)
   > Set CategoryTabsClass!(
    execute
    CategoryTabsClass:WBP UIKit Tabs Base Class Reference=Get CategoryTabGroupClass() -> (CategoryTabGroupClass:WBP UIKit Tabs Base Class Reference)
    self:WBP UIKit Tab Category Base Object Reference=@n3
   ) -> (then, Output_Get:WBP UIKit Tabs Base Class Reference)
   > Set Skip Expand Animation!(execute, Skip Expand Animation:bool=Get Skip Expand Animation(), self:WBP UIKit Tab Category Base Object Reference=@n3) -> (then, Output_Get:bool)
   > Set ExpandDuration!(
    execute
    ExpandDuration:float=Get ExpandAnimationDuration() -> (ExpandAnimationDuration:float)
    self:WBP UIKit Tab Category Base Object Reference=@n3
   ) -> (then, Output_Get:float)
   > SetBackground!(
    execute
    self:WBP UIKit Tab Category Base Object Reference=@n3
    InBrush:"Slate Brush Structure (by ref)"=Get TabGroupBackground() -> (TabGroupBackground:Slate Brush Structure)
   ) -> (then)
   > Bind Event to On Category Tab Selected(self:WBP UIKit Tab Category Base Object Reference=@n3, Delegate=Create Event()) {is=title K2Node_AddDelegate}
   > Bind Event to On Category Tab Clicked(self:WBP UIKit Tab Category Base Object Reference=@n3, Delegate=Create Event()) {is=title K2Node_AddDelegate}
   > Bind Event to On Category Tabs Creation(self:WBP UIKit Tab Category Base Object Reference=@n3, Delegate=Create Event()) {is=title K2Node_AddDelegate}
   > Bind Event to On Category Expanded State Changed(self:WBP UIKit Tab Category Base Object Reference=@n3, Delegate=Create Event()) {is=title K2Node_AddDelegate}
   > Return Node!(execute, Container:WBP UIKit Tab Category Base Object Reference=Get CategoryWidget())
  ]}
 EventGraph {nodes [
   n1 = Event OnCategoryExpandedStateChanged!() -> (OutputDelegate, then, CategoryID:Name, IsExpanded:bool=~) {is=title K2Node_CustomEvent}
   > HandleCategoryExpandedStateChanged!(execute, self, CategoryID:Name=@n1.CategoryID, IsExpanded:bool=@n1.IsExpanded) -> (then)
   n2 = Event OnCategoryTabSelected!() -> (OutputDelegate, then, CategoryID:Name, TabID:Name, TabVM:UI Kit Tab Group Button View Model Object Reference) {is=title K2Node_CustomEvent}
   > n3 = Sequence!(execute) -> (then_0, then_1)
   > DeselectAll(self:UIKit Tab Group Object Reference)
   > CollapseOthers!(execute, self, ExceptCategoryID:Name=@n2.CategoryID) -> (then) {is=title K2Node_CallFunction}
   > ExpandCategory!(execute, self, CategoryID:Name=@n2.CategoryID) -> (then)
   > CheckCategoryChanged!(execute, self, CategoryID:Name=@n2.CategoryID) -> (then) {is=title K2Node_CallFunction}
   Call On Category Tab Selected!(
    execute:Exec=@n3.then_1
    self:WBP UIKit Tabs Categories Base Object Reference
    CategoryID:Name=@n2.CategoryID
    TabID:Name=@n2.TabID
    TabVM:UI Kit Tab Group Button View Model Object Reference=@n2.TabVM
   ) -> (then:Exec) {is=title K2Node_CallDelegate}
   n4 = Event OnTabClicked!() -> (OutputDelegate, then, TabId:Name) {is=title K2Node_CustomEvent}
   > HandleCategoryClick!(execute, self, CategoryID:Name=@n4.TabId) -> (then)
   > NotifyCategoryClicked!(execute, self, TabID:Name=@n4.TabId) -> (then)
   > Branch(Condition=NOT Boolean(A=Get AllowMultipleExpandedCategories() -> (AllowMultipleExpandedCategories:bool)))
   > CollapseOthers(ExceptCategoryID=@n4.TabId)
   n5 = Event OnTabSelected!() -> (OutputDelegate, then, TabId:Name) {is=title K2Node_CustomEvent}
   > Branch(Condition="Not Equal (Name)"(A=Get Previous Selected Category() -> (Previous Selected Category=None), B=@n5.TabId) -> (ReturnValue=~))
   > CheckCategoryChanged(CategoryID=@n5.TabId)
   > Branch(Condition=NOT Boolean(A=Get AllowMultipleExpandedCategories()))
   > CollapseOthers(ExceptCategoryID=@n5.TabId)
   n6 = Event OnCategoryTabsInitialised!() -> (OutputDelegate, then, Tabs:WBP UIKit Tabs Base Object Reference) {is=title K2Node_CustomEvent}
   > CopyTabsProps!(execute, self, Target:WBP UIKit Tabs Base Object Reference=@n6.Tabs) -> (then) {is=title K2Node_CallFunction}
   LatentDeselectAllCategories() {is=title K2Node_CustomEvent}
   > DelayUntilNextTick!(execute) -> (then)
   > DeselectAll()
   n7 = Event OnCategoryTabClicked!() -> (OutputDelegate, then, CategoryID:Name, TabID:Name, TabVM:UI Kit Tab Group Button View Model Object Reference) {is=title K2Node_CustomEvent}
   > Call On Category Tab Clicked!(
    execute:Exec
    self:WBP UIKit Tabs Categories Base Object Reference
    CategoryID:Name=@n7.CategoryID
    TabID:Name=@n7.TabID
    TabVM:UI Kit Tab Group Button View Model Object Reference=@n7.TabVM
   ) -> (then:Exec) {is=title K2Node_CallDelegate}
   n8 = Event PreConstruct()
   > n9 = Sequence()
   > Clear(TargetArray:Array of WBP UIKit Tab Category Base Object References=Get CategoryContainers())
   > Clear(TargetArray:Array of UI Kit Tab Group Button View Model Object References=Get LocalTabsViewModels() -> (LocalTabsViewModels:Array of UI Kit Tab Group Button View Model Object References))
   > SetCategoriesData!(
    execute
    self
    Categories:Array of UI Kit Tab Group Category View Model Object References=Get Local Categories View Models() -> (Local Categories View Models:Array of UI Kit Tab Group Category View Model Object References)
   ) -> (then)
   Parent: PreConstruct(execute=@n9.then_1, IsDesignTime=@n8.IsDesignTime)
   > Bind Event to On Tab Selected(self:Common Tab List Widget Base Object Reference, Delegate=Create Event()) {is=title K2Node_AddDelegate}
   > Bind Event to On Tab Clicked(self:UIKit Tab Group Object Reference, Delegate=Create Event()) {is=title K2Node_AddDelegate}
   Event Construct()
   > Parent: Construct()
   n10 = Event Event On Input Method Changed!() -> (OutputDelegate, then, bNewInputType:ECommonInputType Enum) {is=title K2Node_Event}
   > Parent: Event On Input Method Changed!(execute, bNewInputType:ECommonInputType Enum=@n10.bNewInputType) -> (then) {is=title K2Node_CallParentFunction}
   > RefreshFocusOnSelectionState!(execute, self, InputType:ECommonInputType Enum=@n10.bNewInputType) -> (then)
   Event Destruct()
   > Parent: Destruct()
   > Unbind all Events from On Category Tab Selected!(execute:Exec, self:WBP UIKit Tabs Categories Base Object Reference) -> (then:Exec) {is=title K2Node_ClearDelegate}
   > Unbind all Events from On Category Button Clicked!(execute:Exec, self:WBP UIKit Tabs Categories Base Object Reference) -> (then:Exec) {is=title K2Node_ClearDelegate}
   > Unbind all Events from On Category Tab Clicked!(execute:Exec, self:WBP UIKit Tabs Categories Base Object Reference) -> (then:Exec) {is=title K2Node_ClearDelegate}
  ]}
 ExpandCategory {nodes [
   n1 = ExpandCategory!() -> (then, CategoryID:Name) {is=title K2Node_FunctionEntry}
   > n2 = Set CategoryContainer(CategoryContainer=GetCategoryContainer(CategoryID=@n1))
   > Is Valid(InputObject=@n2)
   n3 = Get CategoryContainer() -> (CategoryContainer:WBP UIKit Tab Category Base Object Reference)
   > n4 = Branch(Condition=Get Is Expanded!(self:WBP UIKit Tab Category Base Object Reference=@n3) -> (Is Expanded:bool))
   SetIsExpanded(execute=@n4.else, self=@n3, InIsExpanded=true, SkipAnim=Get Skip Expand Animation())
  ]}
 GetCategoryContainer {nodes [
   GetCategoryContainer!() -> (then, CategoryID:Name) {is=title K2Node_FunctionEntry}
   > n1 = For Each Loop(Array=Get CategoryContainers())
   Return Node(execute=@n1.Completed)
   > Branch(Condition="Equal (Name)"(A=Get CategoryID(self=@n1) -> (CategoryID=None), B=Get CategoryID!() -> (CategoryID=None)))
   > Return Node(Container=@n1)
  ]}
 GetCategorySelectedTabButton {nodes [
   n1 = GetCategorySelectedTabButton!() -> (then, CategoryID:Name) {is=title K2Node_FunctionEntry}
   > n2 = Set CategoryContainer(CategoryContainer=GetCategoryContainer(CategoryID=@n1) {is=title K2Node_CallFunction})
   > n3 = Is Valid(InputObject=@n2)
   > n4 = Set CategoryTabs(CategoryTabs:WBP UIKit Tabs Base Object Reference=GetCategoryTabs!(self:WBP UIKit Tab Category Base Object Reference=@n2) -> (InnerTabs:WBP UIKit Tabs Base Object Reference)) -> (Output_Get:WBP UIKit Tabs Base Object Reference)
   > n5 = Is Valid(InputObject=@n4)
   > Return Node!(
    execute
    ReturnValue:Common Button Base Object Reference=GetTabButtonBaseByID!(
     self:Common Tab List Widget Base Object Reference=@n4
     TabNameID:Name=GetSelectedTabId!(self:Common Tab List Widget Base Object Reference=@n4) -> (ReturnValue:Name=None)
    ) -> (ReturnValue:Common Button Base Object Reference)
   )
   Return Node(execute=[@n5.Is Not Valid, @n3.Is Not Valid])
  ]}
 GetCategoryTabButton {nodes [
   n1 = GetCategoryTabButton!() -> (then, CategoryID:Name, TabNameID:Name) {is=title K2Node_FunctionEntry}
   > n2 = Set CategoryContainer(CategoryContainer=GetCategoryContainer(CategoryID=@n1))
   > n3 = Is Valid(InputObject=@n2)
   > n4 = Set CategoryTabs(CategoryTabs=GetCategoryTabs(self=@n2))
   > n5 = Is Valid(InputObject=@n4)
   > Return Node(ReturnValue=GetTabButtonBaseByID(self=@n4, TabNameID=Get TabNameID() -> (TabNameID:Name=None)))
   Return Node(execute=[@n5.Is Not Valid, @n3.Is Not Valid])
  ]}
 GetCategoryViewModelByID {nodes [
   n1 = GetCategoryViewModelByID!() -> (then, CategoryID:Name) {is=title K2Node_FunctionEntry}
   > n2 = Set TabVM(TabVM:UI Kit Tab Group Button View Model Object Reference=GetTabViewModelByID!(self:UIKit Tab Group Object Reference, TabNameID:Name=@n1) -> (ReturnValue:UI Kit Tab Group Button View Model Object Reference)) -> (Output_Get:UI Kit Tab Group Button View Model Object Reference)
   > n3 = Is Valid(InputObject=@n2)
   Return Node!(execute=[@n4.Completed, @n3.Is Not Valid], Category:UI Kit Tab Group Category View Model Object Reference)
   > n4 = For Each Loop(Array:Array of UI Kit Tab Group Category View Model Object References=Get Local Categories View Models()) -> (Array Element:UI Kit Tab Group Category View Model Object Reference)
   > Branch(Condition="Equal (Object)"(
     A=Get Category!(self:UI Kit Tab Group Category View Model Object Reference=@n4) -> (Category:UI Kit Tab Group Button View Model Object Reference)
     B=Get TabVM() -> (TabVM:UI Kit Tab Group Button View Model Object Reference)
    ))
   > Return Node(Category=@n4)
  ]}
 GetCategoryViewModelByIndex {nodes [
   n1 = GetCategoryViewModelByIndex!() -> (then, Index:int) {is=title K2Node_FunctionEntry}
   > n2 = Branch(Condition=Is Valid Index!(TargetArray:Array of UI Kit Tab Group Category View Model Object References=Get Local Categories View Models(), IndexToTest:int=@n1) -> (ReturnValue:bool))
   > Return Node(Category="Get (a copy)"(
     Array:Array of UI Kit Tab Group Category View Model Object References=Get Local Categories View Models()
     Dimension 1=Get Index() -> (Index=0)
    ) -> (Output:UI Kit Tab Group Category View Model Object Reference))
   Return Node(execute=@n2.else)
  ]}
 HandleCategoryClick {nodes [
   n1 = HandleCategoryClick!() -> (then, CategoryID:Name) {is=title K2Node_FunctionEntry}
   > n2 = Set LocalCategoryContainer(LocalCategoryContainer:WBP UIKit Tab Category Base Object Reference=GetCategoryContainer(CategoryID=@n1)) -> (Output_Get:WBP UIKit Tab Category Base Object Reference)
   > Is Valid(InputObject=@n2)
   > Set LocalInputMethod(LocalInputMethod:ECommonInputType Enum=GetCurrentInputType(self=Get CommonInputSubsystem())) -> (Output_Get:ECommonInputType Enum)
   > n3 = Branch(Condition=Get Is Expanded(self=Get LocalCategoryContainer() -> (LocalCategoryContainer:WBP UIKit Tab Category Base Object Reference)))
   > Branch(Condition="Not Equal (Enum)"!(A:ECommonInputType Enum=Get LocalInputMethod() -> (LocalInputMethod:ECommonInputType Enum), B:ECommonInputType Enum=Gamepad) -> (ReturnValue:bool))
   > SetIsExpanded(self=Get LocalCategoryContainer(), InIsExpanded=false, SkipAnim=Get Skip Expand Animation())
   n4 = Sequence(execute=@n3.else)
   > Branch(Condition="Equal (Enum)"(A=Get LocalInputMethod(), B=Gamepad))
   n5 = GetCategoryTabs(self=Get LocalCategoryContainer())
   > Is Valid(InputObject=@n5)
   > SelectTabByIndex!(execute, self:UIKit Tab Group Object Reference=@n5, Index:int=0, bSuppressClickFeedback:bool=true) -> (then, ReturnValue:bool)
   ExpandToSelectedItem(execute=@n4.then_1, self:WBP UIKit Tab Category Base Object Reference=Get LocalCategoryContainer())
  ]}
 HandleCategoryExpandedStateChanged {nodes [
   n1 = HandleCategoryExpandedStateChanged!() -> (then, CategoryID:Name, IsExpanded:bool=~) {is=title K2Node_FunctionEntry}
   > n2 = Set LocalCategoryContainer(LocalCategoryContainer=GetCategoryContainer(CategoryID=@n1))
   > Is Valid(InputObject=@n2)
   > n3 = Branch(Condition=@n1.IsExpanded)
   n4 = GetCategoryTabs(self=Get LocalCategoryContainer())
   > Branch(Condition="integer >= integer"(A=GetSelectedTabIndex!(self:UIKit Tab Group Object Reference=@n4) -> (ReturnValue:int=0)))
   > LatentDeselectAllCategories() {is=title K2Node_CallFunction}
   Branch(execute=@n3.else, Condition=ShouldDisplaySelectionOnCollapsedCategory!(self) -> (ReturnValue:bool))
   n5 = GetCategoryTabs(self=Get LocalCategoryContainer())
   > Branch(Condition="integer >= integer"(A=GetSelectedTabIndex(self=@n5) -> (ReturnValue=0)))
   > n6 = Set LocalCategoryButton(LocalCategoryButton:Common Button Base Object Reference=GetTabButtonBaseByID(TabNameID=Get CategoryID() -> (CategoryID=None))) -> (Output_Get:Common Button Base Object Reference)
   > Is Valid(InputObject=@n6)
   n7 = Get LocalCategoryButton() -> (LocalCategoryButton:Common Button Base Object Reference)
   > SetIsSelectable!(execute, self:Common Button Base Object Reference=@n7, bInIsSelectable:bool=true) -> (then)
   > SetIsSelected!(execute, self:Common Button Base Object Reference=@n7, InSelected:bool=true, bGiveClickFeedback:bool=false) -> (then)
  ]}
 IsCategoryExpanded {nodes [
   n1 = IsCategoryExpanded!() -> (then, CategoryID:Name) {is=title K2Node_FunctionEntry}
   > n2 = Set CategoryContainer(CategoryContainer=GetCategoryContainer(CategoryID=@n1))
   > n3 = Is Valid(InputObject=@n2)
   Return Node!(execute=@n3.Is Not Valid, IsExpanded:bool=true)
   > Return Node(IsExpanded=Get Is Expanded(self=@n2))
  ]}
 IsCategoryViewModelValid {nodes [
   IsCategoryViewModelValid!() -> (then, CategoryVM:UI Kit Tab Group Category View Model Object Reference) {is=title K2Node_FunctionEntry}
   > n1 = Sequence()
   n2 = Get CategoryVM() -> (CategoryVM:UI Kit Tab Group Category View Model Object Reference)
   > Is Valid(InputObject=@n2)
   > Is Valid(InputObject=Get Category(self=@n2))
   n3 = Length(TargetArray:Array of UI Kit Tab Group Button View Model Object References=Get Tabs(self=@n2))
   > Branch(Condition=OR Boolean(A="integer > integer"(A=@n3), B=Get AllowEmptyCategories()))
   > Return Node!(execute, IsValid:bool=true)
   PrintString!(
    execute=@n1.then_1
    InString=WBP_UIKit_Tabs_Categories_Base::IsCategoryViewModeValid - Invalid VM
    bPrintToScreen
    bPrintToLog
    TextColor
    Duration
    Key
   ) -> (then)
   > Return Node(IsValid=false)
  ]}
 NotifyCategoryClicked {nodes [
   n1 = NotifyCategoryClicked!() -> (then, TabID:Name) {is=title K2Node_FunctionEntry}
   > n2 = Set TabButton(TabButton:Common Button Base Object Reference=GetTabButtonBaseByID(TabNameID=@n1)) -> (Output_Get:Common Button Base Object Reference)
   > Is Valid(InputObject=@n2)
   > n3 = Set CategoryContainer(CategoryContainer=GetCategoryContainer(CategoryID=Get TabID() -> (TabID:Name=None)))
   > n4 = Is Valid(InputObject=@n3)
   > Call On Category Button Clicked!(
    execute:Exec
    self:WBP UIKit Tabs Categories Base Object Reference
    Button:Common Button Base Object Reference=Get TabButton() -> (TabButton:Common Button Base Object Reference)
    CategoryID:Name=Get TabID() -> (TabID=None)
    IsCategoryExpanded:bool=Get Is Expanded(self=Get CategoryContainer())
   ) -> (then:Exec) {is=title K2Node_CallDelegate}
   Call On Category Button Clicked(execute=@n4.Is Not Valid, Button=Get TabButton(), CategoryID=Get TabID() -> (TabID=None), IsCategoryExpanded=true)
  ]}
 OnCategoryButtonClicked {nodes [OnCategoryButtonClicked!() -> (then, Button:Common Button Base Object Reference, CategoryID:Name, IsCategoryExpanded:bool=~) {is=title K2Node_FunctionEntry}]}
 OnCategoryTabClicked {nodes [OnCategoryTabClicked!() -> (then, CategoryID:Name, TabID:Name, TabVM:UI Kit Tab Group Button View Model Object Reference) {is=title K2Node_FunctionEntry}]}
 OnCategoryTabSelected {nodes [OnCategoryTabSelected!() -> (then, CategoryID:Name, TabID:Name, TabVM:UI Kit Tab Group Button View Model Object Reference) {is=title K2Node_FunctionEntry}]}
 RefreshCategorySelectableState {nodes [
   n1 = RefreshCategorySelectableState!() -> (then, CategoryID:Name, InputType:ECommonInputType Enum) {is=title K2Node_FunctionEntry}
   > n2 = Set LocalButton(LocalButton:Common Button Base Object Reference=GetTabButtonBaseByID(TabNameID=@n1)) -> (Output_Get:Common Button Base Object Reference)
   > Is Valid(InputObject=@n2)
   > n3 = Set LocalShouldSelectOnFocus(LocalShouldSelectOnFocus:bool="Equal (Enum)"(A=Get InputType() -> (InputType:ECommonInputType Enum), B=Gamepad)) -> (Output_Get:bool)
   n4 = IsValid(Object=GetCategoryContainer(CategoryID=@n1))
   n5 = Get LocalButton() -> (LocalButton:Common Button Base Object Reference)
   > SetIsSelectable(self=@n5, bInIsSelectable=OR Boolean(A=@n3, B=NOT Boolean(A=@n4)))
   > SetShouldSelectUponReceivingFocus!(
    execute
    self:Common Button Base Object Reference=@n5
    bInShouldSelectUponReceivingFocus:bool=Get LocalShouldSelectOnFocus() -> (LocalShouldSelectOnFocus:bool)
   ) -> (then)
  ]}
 RefreshFocusOnSelectionState {nodes [
   n1 = RefreshFocusOnSelectionState!() -> (then, InputType:ECommonInputType Enum) {is=title K2Node_FunctionEntry}
   > Set LocalShouldSelectOnFocus(LocalShouldSelectOnFocus="Equal (Enum)"(A=@n1, B=Gamepad))
   > n2 = For Loop(LastIndex=int - int(A=GetTabCount!(self:Common Tab List Widget Base Object Reference) -> (ReturnValue:int=0)))
   > n3 = Set LocalCategoryID(LocalCategoryID:Name=GetTabIdAtIndex!(self:Common Tab List Widget Base Object Reference, Index:int=@n2) -> (ReturnValue:Name=None)) -> (Output_Get:Name=None)
   > RefreshCategorySelectableState(CategoryID=@n3, InputType=Get InputType()) {is=title K2Node_CallFunction}
   > n4 = Set LocalCategoryContainer(LocalCategoryContainer=GetCategoryContainer(CategoryID=@n3))
   > Is Valid(InputObject=@n4)
   n5 = GetCategoryTabs(self=@n4)
   > Is Valid(InputObject=@n5)
   > n6 = For Loop(LastIndex=int - int(A=GetTabCount(self=@n5) -> (ReturnValue=0)))
   > n7 = Set LocalButton(LocalButton=GetTabButtonBaseByID(self=@n5, TabNameID=GetTabIdAtIndex(self=@n5, Index=@n6) -> (ReturnValue=None)))
   > Is Valid(InputObject=@n7)
   > SetShouldSelectUponReceivingFocus(self=@n7, bInShouldSelectUponReceivingFocus=Get LocalShouldSelectOnFocus())
  ]}
 SetCategoriesData {nodes [
   n1 = SetCategoriesData!() -> (then, Categories:Array of UI Kit Tab Group Category View Model Object References) {is=title K2Node_FunctionEntry}
   > n2 = Set Local Categories View Models(Local Categories View Models:Array of UI Kit Tab Group Category View Model Object References=@n1) -> (Output_Get:Array of UI Kit Tab Group Category View Model Object References)
   > Clear(TargetArray=Get TabsData() -> (TabsData:Array of UI Kit Tab Group Button View Model Object References))
   > Clear(TargetArray:Array of WBP UIKit Tab Category Base Object References=Get CategoryContainers())
   > Set Previous Selected Category(Previous Selected Category=None) -> (Output_Get=None)
   > n3 = For Each Loop(Array=@n2)
   Set Tabs Data!(
    execute=@n3.Completed
    self:WBP UIKit Tabs Base Object Reference
    InTabs Data:Array of UI Kit Tab Group Button View Model Object References=Get TabsData()
   ) -> (then) {is=title K2Node_CallFunction}
   > RefreshFocusOnSelectionState(InputType=GetCurrentInputType(self=Get CommonInputSubsystem())) {is=title K2Node_CallFunction}
   Branch(execute=@n3.LoopBody, Condition=IsCategoryViewModelValid(CategoryVM=@n3) {is=title K2Node_CallFunction})
   > Add(
    TargetArray:Array of UI Kit Tab Group Button View Model Object References=Get TabsData()
    NewItem:"UI Kit Tab Group Button View Model Object Reference (by ref)"=Get Category(self=@n3)
   )
  ]}
 ShouldDisplaySelectionOnCollapsedCategory {nodes [
   ShouldDisplaySelectionOnCollapsedCategory() {is=title K2Node_FunctionEntry}
   n1 = Get CommonInputSubsystem()
   > Return Node!(execute, ReturnValue:bool="Not Equal (Enum)"(A=GetCurrentInputType(self=@n1), B=Gamepad))
  ]}
]
variables [
 AllowEmptyCategories {type=boolean}
 AllowMultipleExpandedCategories {type=boolean}
 CategoryButtonClass {type=Class, default=null}
 CategoryClass {type=Class, default=&/WBP_UIKit_TabCategory_Base.WBP_UIKit_TabCategory_Base_C}
 CategoryContainers {type="array<WBP_UIKit_TabCategory_Base_C>", default []}
 CategoryTabGroupClass {type=Class, default=/FortUILibrary/UIKit/Tabs/WBP_UIKit_Tabs.WBP_UIKit_Tabs_C}
 ExpandAnimationDuration {type=number, default=0.25}
 Local Categories View Models {type="array<UIKitTabGroupCategoryViewModel>", default []}
 OnCategoryButtonClicked {}
 OnCategoryTabClicked {}
 OnCategoryTabSelected {}
 Previous Selected Category {type=string, default=None}
 Skip Expand Animation {type=boolean}
 TabGroupBackground {
  type=SlateBrush
  default {
   bIsDynamicallyLoaded=false
   drawAs=NoDrawType
   imageType=NoImage
   mirroring=NoMirror
   resourceName=None
   resourceObject=null
   tiling=NoTile
   imageSize=<32 32>
   margin=<ltrb 0 0 0 0>
   outlineSettings {
    bUseBrushTransparency=false
    roundingType=HalfHeightRadius
    width=0
    color {colorUseRule=UseColor_Specified, specifiedColor=<rgba 0 0 0 0>}
    cornerRadii=<0 0 0 1>
   }
   tintColor {colorUseRule=UseColor_Specified, specifiedColor=<rgba 1 1 1 0>}
   uVRegion {bIsValid=false, max=<0 0>, min=<0 0>}
  }
 }
]
widgets {
 info {inheritedWidgetCount=0, parentClass=&WBP_UIKit_Tabs_C, rootWidgetClass=null, widgetCount=0}
 widgets []
}
