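# UIKit widget for two-level category tabs, wrapping a configurable inner categories widget that it creates in a named slot on PreConstruct and removes on Destruct.
# It forwards category data and clicks, and exposes lookups and selection of categories and their nested tabs by ID or index.
# Selection changes fire one deduplicated OnSelectionChanged event.
# A wait flag swallows the intermediate category callback when code selects category and tab together.
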WidgetBlueprint /FortUILibrary/UIKit/Tabs/WBP_UIKit_Tabs_Categories
graphs [
 AddCategoriesWidget {nodes [
   AddCategoriesWidget() {is=title K2Node_FunctionEntry}
   > n1 = Create Widget!(
    execute:Exec
    Class:User Widget Class Reference=Get CategoriesClass() -> (CategoriesClass:WBP UIKit Tabs Categories Base Class Reference)
    OwningPlayer:Player Controller Object Reference=GetOwningPlayer()
    UIKitTabGroupViewModel:UI Kit Tab Group View Model Object Reference
    bAutoListenForInput:bool=False
    bShouldWrapNavigation:bool=True
   ) -> (then:Exec, ReturnValue:WBP UIKit Tabs Categories Base Object Reference)
   > n2 = Set CategoriesWidget(CategoriesWidget:WBP UIKit Tabs Categories Base Object Reference=@n1) -> (Output_Get:WBP UIKit Tabs Categories Base Object Reference)
   > Is Valid(InputObject=@n2)
   > AddChild(self=Get Slot_Categories() -> (Slot_Categories:Named Slot Object Reference), Content=@n2)
   n3 = Get CategoriesWidget() -> (CategoriesWidget:WBP UIKit Tabs Categories Base Object Reference)
   > Bind Event to On Tab Selected(self:Common Tab List Widget Base Object Reference=@n3, Delegate=Create Event()) {is=title K2Node_AddDelegate}
   > Bind Event to On Tab Clicked(self:UIKit Tab Group Object Reference=@n3, Delegate=Create Event()) {is=title K2Node_AddDelegate}
   > Bind Event to On Category Tab Selected(self:WBP UIKit Tabs Categories Base Object Reference=@n3, Delegate=Create Event()) {is=title K2Node_AddDelegate}
   > Bind Event to On Category Tab Clicked(self:WBP UIKit Tabs Categories Base Object Reference=@n3, Delegate=Create Event()) {is=title K2Node_AddDelegate}
   > SetCategoriesData!(
    execute
    self:WBP UIKit Tabs Categories Base Object Reference=Get CategoriesWidget()
    Categories:Array of UI Kit Tab Group Category View Model Object References=Get LocalCategoriesViewModels() -> (LocalCategoriesViewModels:Array of UI Kit Tab Group Category View Model Object References)
   ) -> (then)
  ]}
 EventGraph {nodes [
   n1 = Event OnTabSelected!() -> (OutputDelegate, then, TabId:Name) {is=title K2Node_CustomEvent}
   > NotifySelectionChanged!(execute, self, CategoryID:Name=@n1.TabId, CategoryTabID:Name=GetSelectedCategoryTabID!(self, CategoryID:Name=@n1.TabId) -> (TabID:Name=None)) -> (then)
   n2 = Event OnCategoryTabSelected!() -> (OutputDelegate, then, CategoryID:Name, TabID:Name, TabVM:UI Kit Tab Group Button View Model Object Reference) {is=title K2Node_CustomEvent}
   > NotifySelectionChanged(CategoryID=@n2.CategoryID, CategoryTabID=@n2.TabID)
   n3 = Event OnCategoryButtonClicked!() -> (OutputDelegate, then, TabId:Name) {is=title K2Node_CustomEvent}
   n4 = Get CategoriesWidget()
   > Call On Category Clicked!(
    execute:Exec
    self:WBP UIKit Tabs Categories Object Reference
    Button:Common Button Base Object Reference=GetTabButtonBaseByID!(self:Common Tab List Widget Base Object Reference=@n4, TabNameID:Name=@n3.TabId) -> (ReturnValue:Common Button Base Object Reference)
    CategoryID:Name=@n3.TabId
    IsCategoryExpanded:bool=IsCategoryExpanded!(self:WBP UIKit Tabs Categories Base Object Reference=@n4, CategoryID:Name=@n3.TabId) -> (IsExpanded:bool)
    CategoryVM:UI Kit Tab Group Category View Model Object Reference=GetCategoryViewModelByID!(self:WBP UIKit Tabs Categories Base Object Reference=@n4, CategoryID:Name=@n3.TabId) -> (Category:UI Kit Tab Group Category View Model Object Reference=UIKit Tabs)
   ) -> (then:Exec) {is=title K2Node_CallDelegate}
   n5 = Event OnCategoryTabClicked!() -> (OutputDelegate, then, CategoryID:Name, TabID:Name, TabVM:UI Kit Tab Group Button View Model Object Reference) {is=title K2Node_CustomEvent}
   > Call On Category Tab Clicked!(
    execute:Exec
    self:WBP UIKit Tabs Categories Object Reference
    CategoryID:Name=@n5.CategoryID
    TabID:Name=@n5.TabID
    TabVM:UI Kit Tab Group Button View Model Object Reference=@n5.TabVM
   ) -> (then:Exec) {is=title K2Node_CallDelegate}
   n6 = Event PreConstruct()
   > Parent: PreConstruct(IsDesignTime=@n6.IsDesignTime)
   > AddCategoriesWidget() {is=title K2Node_CallFunction}
   Event Construct()
   > Parent: Construct()
   Event Destruct()
   > Parent: Destruct()
   > Unbind all Events from On Category Clicked!(execute:Exec, self:WBP UIKit Tabs Categories Object Reference) -> (then:Exec) {is=title K2Node_ClearDelegate}
   > Unbind all Events from On Category Tab Clicked!(execute:Exec, self:WBP UIKit Tabs Categories Object Reference) -> (then:Exec) {is=title K2Node_ClearDelegate}
   > Unbind all Events from On Selection Changed!(execute:Exec, self:WBP UIKit Tabs Categories Object Reference) -> (then:Exec) {is=title K2Node_ClearDelegate}
   > RemoveCategoriesWidget()
   Get CategoriesClass()
  ]}
 GetCategoryIDAtIndex {nodes [
   n1 = GetCategoryIDAtIndex!() -> (then, Index:int) {is=title K2Node_FunctionEntry}
   n2 = Get CategoriesWidget()
   > n3 = Is Valid(InputObject=@n2)
   Return Node!(execute=@n3.Is Not Valid, CategoryID:Name=None)
   > Return Node(CategoryID=GetTabIdAtIndex!(self:Common Tab List Widget Base Object Reference=@n2, Index:int=@n1) -> (ReturnValue:Name=None))
  ]}
 GetCategoryIndex {nodes [
   n1 = GetCategoryIndex!() -> (then, CategoryID:Name) {is=title K2Node_FunctionEntry}
   n2 = Get CategoriesWidget()
   > n3 = Is Valid(InputObject=@n2)
   Return Node!(execute=@n3.Is Not Valid, Index:int=-1)
   > Return Node(Index=GetTabIndexByID!(self:UIKit Tab Group Object Reference=@n2, TabNameID:Name=@n1) -> (ReturnValue:int=0))
  ]}
 GetCategoryTabIDAtIndex {nodes [
   n1 = GetCategoryTabIDAtIndex!() -> (then, CategoryID:Name, Index:int) {is=title K2Node_FunctionEntry}
   n2 = Get CategoriesWidget()
   > n3 = Is Valid(InputObject=@n2)
   > n4 = Set ContainerWidget(ContainerWidget:WBP UIKit Tab Category Base Object Reference=GetCategoryContainer!(self:WBP UIKit Tabs Categories Base Object Reference=@n2, CategoryID:Name=@n1) -> (Container:WBP UIKit Tab Category Base Object Reference)) -> (Output_Get:WBP UIKit Tab Category Base Object Reference)
   > n5 = Is Valid(InputObject=@n4)
   Return Node!(execute=[@n5.Is Not Valid, @n3.Is Not Valid], TabID:Name=None)
   n6 = Get ContainerWidget() -> (ContainerWidget:WBP UIKit Tab Category Base Object Reference)
   > Return Node(TabID=GetTabIdAtIndex(
     self=GetCategoryTabs!(self:WBP UIKit Tab Category Base Object Reference=@n6) -> (InnerTabs:WBP UIKit Tabs Base Object Reference)
     Index=Get Index() -> (Index:int=0)
    ) -> (ReturnValue=None))
  ]}
 GetCategoryTabIndex {nodes [
   n1 = GetCategoryTabIndex!() -> (then, CategoryID:Name, TabID:Name) {is=title K2Node_FunctionEntry}
   n2 = Get CategoriesWidget()
   > n3 = Is Valid(InputObject=@n2)
   > n4 = Set ContainerWidget(ContainerWidget=GetCategoryContainer(self=@n2, CategoryID=@n1))
   > n5 = Is Valid(InputObject=@n4)
   Return Node!(execute=[@n5.Is Not Valid, @n3.Is Not Valid], Index:int=-1)
   n6 = Get ContainerWidget()
   > Return Node(Index=GetTabIndexByID(self=GetCategoryTabs(self=@n6), TabNameID=Get TabID() -> (TabID:Name=None)) -> (ReturnValue=0))
  ]}
 GetCategoryTabsCount {nodes [
   n1 = GetCategoryTabsCount!() -> (then, CategoryID:Name) {is=title K2Node_FunctionEntry}
   n2 = Get CategoriesWidget()
   > n3 = Is Valid(InputObject=@n2)
   > n4 = Set ContainerWidget(ContainerWidget=GetCategoryContainer(self=@n2, CategoryID=@n1))
   > n5 = Is Valid(InputObject=@n4)
   Return Node!(execute=[@n5.Is Not Valid, @n3.Is Not Valid], Count:int=0)
   n6 = Get ContainerWidget()
   > Return Node(Count=GetTabCount!(self:Common Tab List Widget Base Object Reference=GetCategoryTabs(self=@n6)) -> (ReturnValue:int=0))
  ]}
 GetEntryViewModel {nodes [
   n1 = GetEntryViewModel!() -> (then, CategoryID:Name, TabID:Name) {is=title K2Node_FunctionEntry}
   n2 = Get CategoriesWidget()
   > n3 = Is Valid(InputObject=@n2)
   > n4 = Branch(Condition="Equal (Name)"(A=@n1.TabID))
   > Return Node!(
    execute
    VM:UI Kit Tab Group Button View Model Object Reference=GetTabViewModelByID!(self:UIKit Tab Group Object Reference=@n2, TabNameID:Name=Get CategoryID() -> (CategoryID:Name=None)) -> (ReturnValue:UI Kit Tab Group Button View Model Object Reference)
   )
   n5 = Set CategoryContainer(
    execute=@n4.else
    CategoryContainer:WBP UIKit Tab Category Base Object Reference=GetCategoryContainer(self=@n2, CategoryID=Get CategoryID() -> (CategoryID=None))
   ) -> (Output_Get:WBP UIKit Tab Category Base Object Reference)
   > n6 = Is Valid(InputObject=@n5)
   Return Node(execute=[@n8.Is Not Valid, @n6.Is Not Valid, @n3.Is Not Valid])
   n7 = GetCategoryTabs(self=@n5)
   > n8 = Is Valid(InputObject=@n7)
   > Return Node(VM=GetTabViewModelByID(self=@n7, TabNameID=Get TabID() -> (TabID=None)))
  ]}
 GetSelectedCategoryID {nodes [
   GetSelectedCategoryID() {is=title K2Node_FunctionEntry}
   n1 = Get CategoriesWidget()
   > n2 = Set LocalSelectedCategoryID(LocalSelectedCategoryID:Name=GetSelectedTabId!(self:Common Tab List Widget Base Object Reference=@n1) -> (ReturnValue:Name=None)) -> (Output_Get:Name=None)
   > n3 = Branch(Condition="Equal (Name)"(A=@n2))
   > n4 = Set LocalSelectedCategoryID(LocalSelectedCategoryID=Get Previous Selected Category!(self:WBP UIKit Tabs Categories Base Object Reference=@n1) -> (Previous Selected Category:Name=None)) -> (Output_Get=None)
   > Return Node!(execute, CategoryID:Name=@n4)
   Return Node(execute=@n3.else, CategoryID=Get LocalSelectedCategoryID() -> (LocalSelectedCategoryID:Name=None))
  ]}
 GetSelectedCategoryIndex {nodes [
   GetSelectedCategoryIndex() {is=title K2Node_FunctionEntry}
   n1 = Get CategoriesWidget()
   > n2 = Is Valid(InputObject=@n1)
   Return Node!(execute=@n2.Is Not Valid, Index:int=-1)
   > Return Node(Index=GetSelectedTabIndex!(self:UIKit Tab Group Object Reference=@n1) -> (ReturnValue:int=0))
  ]}
 GetSelectedCategoryTabID {nodes [
   n1 = GetSelectedCategoryTabID!() -> (then, CategoryID:Name) {is=title K2Node_FunctionEntry}
   n2 = Get CategoriesWidget()
   > n3 = Is Valid(InputObject=@n2)
   > n4 = Set ContainerWidget(ContainerWidget=GetCategoryContainer(self=@n2, CategoryID=@n1))
   > n5 = Is Valid(InputObject=@n4)
   Return Node!(execute=[@n5.Is Not Valid, @n3.Is Not Valid], TabID:Name=None)
   n6 = Get ContainerWidget()
   > Return Node(TabID=GetSelectedTabId(self=GetCategoryTabs(self=@n6)) -> (ReturnValue=None))
  ]}
 GetSelectedCategoryTabIndex {nodes [
   n1 = GetSelectedCategoryTabIndex!() -> (then, CategoryID:Name) {is=title K2Node_FunctionEntry}
   n2 = Get CategoriesWidget()
   > n3 = Is Valid(InputObject=@n2)
   > n4 = Set ContainerWidget(ContainerWidget=GetCategoryContainer(self=@n2, CategoryID=@n1))
   > n5 = Is Valid(InputObject=@n4)
   Return Node!(execute=[@n5.Is Not Valid, @n3.Is Not Valid], Index:int=-1)
   n6 = Get ContainerWidget()
   > Return Node(Index=GetSelectedTabIndex(self=GetCategoryTabs(self=@n6)) -> (ReturnValue=0))
  ]}
 GetSelectedTabButton {nodes [
   GetSelectedTabButton() {is=title K2Node_FunctionEntry}
   n1 = Get CategoriesWidget()
   > n2 = Is Valid(InputObject=@n1)
   > n3 = Set LocalSelectedCategoryID(LocalSelectedCategoryID=GetSelectedTabId(self=@n1) -> (ReturnValue=None)) -> (Output_Get=None)
   > n4 = Branch(Condition="Equal (Name)"(A=@n3))
   > n5 = Set LocalSelectedCategoryID(LocalSelectedCategoryID=Get Previous Selected Category(self=@n1) -> (Previous Selected Category=None)) -> (Output_Get=None)
   n6 = Get LocalSelectedCategoryID() -> (LocalSelectedCategoryID=None)
   n7 = Set CategoryButton(execute=[@n5.then, @n4.else], CategoryButton:Common Button Base Object Reference=GetTabButtonBaseByID(self=@n1, TabNameID=@n6)) -> (Output_Get:Common Button Base Object Reference)
   > n8 = Is Valid(InputObject=@n7)
   Return Node!(execute=[@n8.Is Not Valid, @n2.Is Not Valid], Button:Common Button Base Object Reference)
   n9 = GetCategoryContainer(self=Get CategoriesWidget(), CategoryID=@n6)
   > n10 = Is Valid(InputObject=@n9)
   n11 = GetCategoryTabs(self=@n9)
   > n12 = Set CategoryTabButton(CategoryTabButton:Common Button Base Object Reference=GetTabButtonBaseByID(self=@n11, TabNameID=GetSelectedTabId(self=@n11) -> (ReturnValue=None))) -> (Output_Get:Common Button Base Object Reference)
   > n13 = Is Valid(InputObject=@n12)
   Return Node(execute=[@n13.Is Not Valid, @n10.Is Not Valid], Button=Get CategoryButton() -> (CategoryButton:Common Button Base Object Reference))
   > Return Node(Button=Get CategoryTabButton() -> (CategoryTabButton:Common Button Base Object Reference))
  ]}
 GetTabButton {nodes [
   n1 = GetTabButton!() -> (then, CategoryID:Name, TabID:Name) {is=title K2Node_FunctionEntry}
   n2 = Get CategoriesWidget()
   > n3 = Is Valid(InputObject=@n2)
   > n4 = Set CategoryButton(CategoryButton=GetTabButtonBaseByID(self=@n2, TabNameID=@n1))
   > n5 = Is Valid(InputObject=@n4)
   Return Node(execute=[@n5.Is Not Valid, @n3.Is Not Valid])
   n6 = GetCategoryContainer(self=Get CategoriesWidget(), CategoryID=Get CategoryID() -> (CategoryID=None))
   > n7 = Is Valid(InputObject=@n6)
   > n8 = Set CategoryTabButton(CategoryTabButton=GetTabButtonBaseByID(self=GetCategoryTabs(self=@n6), TabNameID=Get TabID() -> (TabID=None)))
   > n9 = Is Valid(InputObject=@n8)
   Return Node(execute=[@n9.Is Not Valid, @n7.Is Not Valid], Button=Get CategoryButton())
   > Return Node(Button=Get CategoryTabButton())
  ]}
 IsCategoryExpanded {nodes [
   n1 = IsCategoryExpanded!() -> (then, CategoryID:Name) {is=title K2Node_FunctionEntry}
   n2 = Get CategoriesWidget()
   > n3 = Is Valid(InputObject=@n2)
   > n4 = Set CategoryContainer(CategoryContainer=GetCategoryContainer(self=@n2, CategoryID=@n1))
   > Return Node!(execute, IsExpanded:bool=Get Is Expanded!(self:WBP UIKit Tab Category Base Object Reference=@n4) -> (Is Expanded:bool))
   Return Node(execute=@n3.Is Not Valid, IsExpanded=false)
  ]}
 IsEmptyCategory {nodes [
   n1 = IsEmptyCategory!() -> (then, CategoryID:Name) {is=title K2Node_FunctionEntry}
   > Return Node!(execute, IsEmpty:bool="Equal (Integer)"(A=GetCategoryTabsCount!(self, CategoryID:Name=@n1) -> (Count:int=0) {is=title K2Node_CallFunction}))
  ]}
 NotifySelectionChanged {nodes [
   n1 = NotifySelectionChanged!() -> (then, CategoryID:Name, CategoryTabID:Name) {is=title K2Node_FunctionEntry}
   > n2 = Branch(Condition=Get ShouldWaitForCategoryTabCallback() -> (ShouldWaitForCategoryTabCallback:bool))
   > Set ShouldWaitForCategoryTabCallback(ShouldWaitForCategoryTabCallback:bool=false) -> (Output_Get:bool)
   n3 = Get LastNotifiedCategoryTabID() -> (LastNotifiedCategoryTabID:Name=None)
   n4 = Get LastNotifiedSelectedCategoryID() -> (LastNotifiedSelectedCategoryID:Name=None)
   Branch(
    execute=@n2.else
    Condition=OR Boolean(A="Not Equal (Name)"(A=@n1, B=@n4) -> (ReturnValue=~), B="Not Equal (Name)"(A=@n1.CategoryTabID, B=@n3) -> (ReturnValue=~))
   )
   n5 = Get CategoryTabID() -> (CategoryTabID:Name=None)
   n6 = Get CategoryID() -> (CategoryID=None)
   > Call On Selection Changed!(execute:Exec, self:WBP UIKit Tabs Categories Object Reference, CategoryID:Name=@n6, TabID:Name=@n5) -> (then:Exec) {is=title K2Node_CallDelegate}
   > Set LastNotifiedSelectedCategoryID(LastNotifiedSelectedCategoryID:Name=@n6) -> (Output_Get:Name=None)
   > Set LastNotifiedCategoryTabID(LastNotifiedCategoryTabID:Name=@n5) -> (Output_Get:Name=None)
  ]}
 OnCategoryClicked {nodes [OnCategoryClicked!() -> (
    then
    Button:Common Button Base Object Reference
    CategoryID:Name
    IsCategoryExpanded:bool=~
    CategoryVM:UI Kit Tab Group Category View Model Object Reference
   ) {is=title K2Node_FunctionEntry}]}
 OnCategoryTabClicked {nodes [OnCategoryTabClicked!() -> (then, CategoryID:Name, TabID:Name, TabVM:UI Kit Tab Group Button View Model Object Reference) {is=title K2Node_FunctionEntry}]}
 OnSelectionChanged {nodes [OnSelectionChanged!() -> (then, CategoryID:Name, TabID:Name) {is=title K2Node_FunctionEntry}]}
 RemoveCategoriesWidget {nodes [
   RemoveCategoriesWidget() {is=title K2Node_FunctionEntry}
   > n1 = Sequence()
   n2 = Get CategoriesWidget()
   > Is Valid(InputObject=@n2)
   > Unbind all Events from On Tab Clicked!(execute:Exec, self:UIKit Tab Group Object Reference=@n2) -> (then:Exec) {is=title K2Node_ClearDelegate}
   > Unbind all Events from On Tab Selected!(execute:Exec, self:Common Tab List Widget Base Object Reference=@n2) -> (then:Exec) {is=title K2Node_ClearDelegate}
   > Unbind all Events from On Category Tab Clicked(self:WBP UIKit Tabs Categories Base Object Reference=@n2)
   > Unbind all Events from On Category Tab Selected!(execute:Exec, self:WBP UIKit Tabs Categories Base Object Reference=@n2) -> (then:Exec) {is=title K2Node_ClearDelegate}
   > RemoveFromParent(self=@n2)
   Set CategoriesWidget(execute=@n1.then_1)
  ]}
 SelectCategoryTabByID {nodes [
   SelectCategoryTabByID!() -> (then, CategoryID:Name, TabID:Name, bSuppressClickFeedback:bool=~) {is=title K2Node_FunctionEntry}
   > n1 = Is Valid(InputObject=Get CategoriesWidget())
   > n2 = Branch(Condition="Not Equal (Name)"(A=Get CategoryID() -> (CategoryID=None)) -> (ReturnValue=~))
   n3 = GetSelectedCategoryTabID(CategoryID=Get CategoryID() -> (CategoryID=None)) -> (TabID=None) {is=title K2Node_CallFunction}
   n4 = Get TabID() -> (TabID=None)
   n5 = GetSelectedCategoryID!(self) -> (CategoryID:Name=None) {is=title K2Node_CallFunction}
   n6 = Get CategoryID() -> (CategoryID=None)
   > n7 = Branch(Condition=AND Boolean(A="Equal (Name)"(A=@n6, B=@n5), B="Equal (Name)"(A=@n3, B=@n4)))
   > Return Node!(execute, ReturnValue:bool=true)
   n8 = Get CategoriesWidget()
   n9 = Get TabID() -> (TabID=None)
   n10 = Get CategoryID() -> (CategoryID=None)
   Set ShouldWaitForCategoryTabCallback(
    execute=@n7.else
    ShouldWaitForCategoryTabCallback=IsValid(Object=GetCategoryTabButton!(self:WBP UIKit Tabs Categories Base Object Reference=@n8, CategoryID:Name=@n10, TabNameID:Name=@n9) -> (ReturnValue:Common Button Base Object Reference))
   )
   > n11 = SelectTabByID!(
    execute
    self:Common Tab List Widget Base Object Reference=Get CategoriesWidget()
    TabNameID:Name=Get CategoryID() -> (CategoryID=None)
    bSuppressClickFeedback:bool=Get bSuppressClickFeedback() -> (bSuppressClickFeedback:bool)
   ) -> (then, ReturnValue:bool)
   > n12 = Branch(Condition=@n11)
   > n13 = Branch(Condition="Not Equal (Name)"(A=Get TabID() -> (TabID=None)) -> (ReturnValue=~))
   > n14 = Set CategoryContainer(CategoryContainer=GetCategoryContainer(self=Get CategoriesWidget(), CategoryID=Get CategoryID() -> (CategoryID=None)))
   > n15 = Is Valid(InputObject=@n14)
   > n16 = SelectTabByID(self=GetCategoryTabs(self=@n14), TabNameID=Get TabID() -> (TabID=None), bSuppressClickFeedback=Get bSuppressClickFeedback())
   > Return Node(ReturnValue=@n16)
   Return Node(execute=[@n15.Is Not Valid, @n12.else, @n2.else, @n1.Is Not Valid], ReturnValue=false)
   Return Node(execute=@n13.else, ReturnValue=true)
  ]}
 SelectCategoryTabByIndex {nodes [
   SelectCategoryTabByIndex!() -> (then, CategoryIndex:int=-1, TabIndex:int=-1, bSuppressClickFeedback:bool=~) {is=title K2Node_FunctionEntry}
   > n1 = Is Valid(InputObject=Get CategoriesWidget())
   > n2 = Set CategoryID(CategoryID:Name=GetTabIdAtIndex(self=Get CategoriesWidget(), Index=Get CategoryIndex() -> (CategoryIndex:int=0)) -> (ReturnValue=None)) -> (Output_Get:Name=None)
   > n3 = Branch(Condition="Not Equal (Name)"(A=@n2) -> (ReturnValue=~))
   n4 = GetSelectedCategoryID() -> (CategoryID=None)
   n5 = Get CategoryID() -> (CategoryID=None)
   n6 = GetSelectedCategoryTabIndex!(self, CategoryID:Name=Get CategoryID() -> (CategoryID=None)) -> (Index:int=0) {is=title K2Node_CallFunction}
   n7 = Get TabIndex() -> (TabIndex:int=0)
   > n8 = Branch(Condition=AND Boolean(A="Equal (Name)"(A=@n5, B=@n4), B="Equal (Integer)"(A=@n6, B=@n7)))
   > Return Node(ReturnValue=true)
   n9 = Get CategoryID() -> (CategoryID=None)
   n10 = GetCategoryTabIDAtIndex!(self, CategoryID:Name=@n9, Index:int=Get TabIndex() -> (TabIndex=0)) -> (TabID:Name=None) {is=title K2Node_CallFunction}
   n11 = Get CategoriesWidget()
   Set ShouldWaitForCategoryTabCallback(execute=@n8.else, ShouldWaitForCategoryTabCallback=IsValid(Object=GetCategoryTabButton(self=@n11, CategoryID=@n9, TabNameID=@n10)))
   > n12 = SelectTabByIndex!(
    execute
    self:UIKit Tab Group Object Reference=Get CategoriesWidget()
    Index:int=Get CategoryIndex() -> (CategoryIndex=0)
    bSuppressClickFeedback:bool=Get bSuppressClickFeedback()
   ) -> (then, ReturnValue:bool)
   > n13 = Branch(Condition=@n12)
   n14 = Get TabIndex() -> (TabIndex=0)
   n15 = GetCategoryTabsCount(CategoryID=Get CategoryID() -> (CategoryID=None)) -> (Count=0)
   > n16 = Branch(Condition=AND Boolean(A="integer >= integer"(A=@n14), B="integer < integer"(A=@n14, B=@n15)))
   > n17 = Set CategoryContainer(CategoryContainer=GetCategoryContainer(self=Get CategoriesWidget(), CategoryID=Get CategoryID() -> (CategoryID=None)))
   > n18 = Is Valid(InputObject=@n17)
   > n19 = SelectTabByIndex(self=GetCategoryTabs(self=@n17), Index=Get TabIndex() -> (TabIndex=0), bSuppressClickFeedback=Get bSuppressClickFeedback())
   > Return Node(ReturnValue=@n19)
   Return Node(execute=[@n18.Is Not Valid, @n13.else, @n3.else, @n1.Is Not Valid], ReturnValue=false)
   Return Node(execute=@n16.else, ReturnValue=true)
  ]}
 SetCategoriesData {nodes [
   n1 = SetCategoriesData!() -> (then, Categories:Array of UI Kit Tab Group Category View Model Object References) {is=title K2Node_FunctionEntry}
   n2 = Get CategoriesWidget()
   > Is Valid(InputObject=@n2)
   > SetCategoriesData(self=@n2, Categories=@n1) {is=title K2Node_CallFunction}
  ]}
]
variables [
 CategoriesClass {type=Class, default=/FortUILibrary/Systems/Tabs/WBP_UIKit_Tabs_Categories_Base.WBP_UIKit_Tabs_Categories_Base_C}
 CategoriesWidget {type=WBP_UIKit_Tabs_Categories_Base_C, default=null}
 LastNotifiedCategoryTabID {type=string, default=None}
 LastNotifiedSelectedCategoryID {type=string, default=None}
 OnCategoryClicked {}
 OnCategoryTabClicked {}
 OnSelectionChanged {}
 ShouldWaitForCategoryTabCallback {type=boolean}
]
widgets {
 info {inheritedWidgetCount=0, parentClass=/Script/FortniteUI.UIKitTabGroupCategories, rootWidgetClass=&NamedSlot, widgetCount=1}
 widgets [Slot_Categories {widgetClassPath=&NamedSlot, bIsVariable=true, widget=Slot_Categories}]
}
