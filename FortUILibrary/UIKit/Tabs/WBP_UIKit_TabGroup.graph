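# WBP_UIKit_TabGroup is a UI Kit tab group widget blueprint derived from WBP_UIKit_TabGroup_Base, adding no widgets of its own.
# Its next and previous arrow click overrides call the parent handler, then broadcast the NextArrowClicked and PreviousArrowClicked event dispatchers.
# PreConstruct and Construct only forward to the parent.
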WidgetBlueprint /FortUILibrary/UIKit/Tabs/WBP_UIKit_TabGroup
graphs [
 EventGraph {nodes [
   n1 = Event PreConstruct()
   > Parent: PreConstruct(IsDesignTime=@n1.IsDesignTime)
   Event Construct()
   > Parent: Construct()
   n2 = Event OnNextArrowButtonClicked!() -> (OutputDelegate, then, Button:Common Button Base Object Reference) {is=title K2Node_Event}
   > Parent: OnNextArrowButtonClicked!(execute, Button:Common Button Base Object Reference=@n2.Button) -> (then) {is=title K2Node_CallParentFunction}
   > Call Next Arrow Clicked!(execute:Exec, self:WBP UIKit Tab Group Object Reference) -> (then:Exec) {is=title K2Node_CallDelegate}
   n3 = Event OnPreviousArrowButtonClicked!() -> (OutputDelegate, then, Button:Common Button Base Object Reference) {is=title K2Node_Event}
   > Parent: OnPreviousArrowButtonClicked!(execute, Button:Common Button Base Object Reference=@n3.Button) -> (then) {is=title K2Node_CallParentFunction}
   > Call Previous Arrow Clicked!(execute:Exec, self:WBP UIKit Tab Group Object Reference) -> (then:Exec) {is=title K2Node_CallDelegate}
  ]}
 NextArrowClicked {nodes [NextArrowClicked() {is=title K2Node_FunctionEntry}]}
 PreviousArrowClicked {nodes [PreviousArrowClicked() {is=title K2Node_FunctionEntry}]}
]
variables [
 NextArrowClicked {}
 PreviousArrowClicked {}
]
widgets {
 info {
  inheritedWidgetCount=0
  parentClass=/FortUILibrary/Systems/Tabs/WBP_UIKit_TabGroup_Base.WBP_UIKit_TabGroup_Base_C
  rootWidgetClass=null
  widgetCount=0
 }
 widgets []
}
